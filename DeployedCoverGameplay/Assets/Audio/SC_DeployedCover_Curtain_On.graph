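# Sound cue for the deployed cover gameplay feature, playing a curtain-on sound, likely as the cover's curtain activates.
# A single wave player feeds a modulator that slightly randomizes pitch on each play.
# It routes to the impacts sound class and uses small-object demolition hit attenuation with overridden falloff.
# A shared concurrency rule limits simultaneous instances, stopping the furthest or oldest.
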SoundCue /DeployedCoverGameplay/Assets/Audio/SC_DeployedCover_Curtain_On
root=Modulator_0
nodes [
 {id=Modulator_0, children [WavePlayer_0], properties {pitchMax=1.02, pitchMin=0.98}}
 {id=WavePlayer_0, properties {soundWaveAssetPtr=&/DeployedCover_Curtain_On}}
]
properties {
 attenuationSettings=/Game/Sounds/Attenuation/Fort_Demolition_Hit_Small_Sized_Object_Att
 bOverrideAttenuation=true
 maxDistance=2600
 soundClassObject=&Fort_Impacts
 volumeMultiplier=1
 attenuationOverrides {falloffDistance=2400, stereoSpread=2000, attenuationShapeExtents {x=200}}
 concurrencyOverrides {maxCount=4}
 concurrencySet [/Game/Sounds/Concurrency_Rules/Demo_FourMaximum_StopFurthestOldest_VolumeScale]
}
