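# Sound cue for equipping Deployed Cover, played as player foley.
# Two equip wave players play together through a Mixer, the first at reduced input volume, then pass through a Modulator at its default settings.
# The cue boosts overall volume, uses the generic player-equip attenuation and overrides concurrency to prevent new instances once the limit is reached.
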SoundCue /DeployedCoverGameplay/Assets/Audio/SC_DeployedCover_Equip
root=Modulator_0
nodes [
 {id=Mixer_0, children [WavePlayer_0,WavePlayer_1], properties {inputVolume [0.75,1]}}
 {id=Modulator_0, children [Mixer_0]}
 {id=WavePlayer_0, properties {soundWaveAssetPtr=&/DeployedCover_Equip_01}}
 {id=WavePlayer_1, properties {soundWaveAssetPtr=&/DeployedCover_Equip_02}}
]
properties {
 attenuationSettings=/Game/Sounds/Attenuation/Fort_Player_Equip_Generic_Att
 bOverrideConcurrency=true
 maxDistance=1800
 soundClassObject=&Fort_Player_Foley
 volumeMultiplier=1.25
 concurrencyOverrides {resolutionRule=PreventNew}
}
