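# Sound cue for the deployed cover explosion, layering an impact sound with a destruction sound.
# A mixer combines two modulated branches: one picks randomly among three hit waves with varied pitch and volume, the other between two destruct waves.
# Playback uses the cover-hit stereo attenuation, a large-explosives sound class and the shared weapon explosion concurrency set.
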SoundCue /DeployedCoverGameplay/Assets/Audio/SC_DeployedCover_Explo
root=Mixer_0
nodes [
 {id=Mixer_0, children [Modulator_1,Modulator_0], properties {inputVolume [1,1]}}
 {id=Modulator_0, children [Random_0], properties {pitchMax=1.25, volumeMin=0.85}}
 {id=Modulator_1, children [Random_1]}
 {id=Random_0, children [WavePlayer_5,WavePlayer_0,WavePlayer_2], properties {weights [1,1,1]}}
 {id=Random_1, children [WavePlayer_7,WavePlayer_8], properties {weights [1,1]}}
 {id=WavePlayer_0, properties {soundWaveAssetPtr=&/DeployedCover_Hit_03}}
 {id=WavePlayer_2, properties {soundWaveAssetPtr=&/DeployedCover_Hit_02}}
 {id=WavePlayer_5, properties {soundWaveAssetPtr=&/DeployedCover_Hit_01}}
 {id=WavePlayer_7, properties {soundWaveAssetPtr=&/Deployed_Cover_Destruct_01}}
 {id=WavePlayer_8, properties {soundWaveAssetPtr=&/Deployed_Cover_Destruct_02}}
]
properties {
 attenuationSettings=&/Att_DeployedCover_Hit_Stereo
 maxDistance=7927
 soundClassObject=/Game/Sounds/FortSoundClassesAndModes/Fort_Explosives_Large_PSM
 volumeMultiplier=0.25
 concurrencyOverrides {maxCount=4, resolutionRule=StopOldest, volumeScale=0.5}
 concurrencySet [/Game/Sounds/Concurrency_Rules/Weapon_Explosion_Conc]
}
