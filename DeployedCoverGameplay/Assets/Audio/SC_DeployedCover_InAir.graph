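# Sound cue for the deployed cover gadget in the air, mixing a start one-shot with a looping warning.
# The start branch randomly picks one of three waves, behind a quality node whose fourth slot is empty.
# The warning loop fades in quickly, with Doppler, occlusion and a PreExplo parameter modulating pitch and volume.
# Each branch overrides sound class and attenuation with HRTF spatialization, while concurrency rules cap simultaneous instances.
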SoundCue /DeployedCoverGameplay/Assets/Audio/SC_DeployedCover_InAir
root=Mixer_0
nodes [
 {
  id=Attenuation_0
  children [SoundClass_1]
  properties {
   attenuationSettings=&/DeployedCover_InAir_Att
   attenuationOverrides {
    distanceAlgorithm=Logarithmic
    falloffDistance=1536
    spatializationAlgorithm=SPATIALIZATION_HRTF
    stereoSpread=0
    attenuationShapeExtents {x=768}
   }
  }
 }
 {
  id=Attenuation_7
  children [ModulatorContinuous_1]
  properties {
   attenuationSettings=&Fort_Weapon_Projectile_Medium_Mono_Athena_Att
   bOverrideAttenuation=true
   attenuationOverrides {
    bEnableOcclusion=true
    bEnableReverbSend=false
    distanceAlgorithm=Logarithmic
    falloffDistance=0
    occlusionLowPassFilterFrequency=1000
    occlusionTraceChannel=ECC_GameTraceChannel11
    spatializationAlgorithm=SPATIALIZATION_HRTF
    stereoSpread=0
    attenuationShapeExtents {x=550}
   }
  }
 }
 {id=Doppler_0, children [Attenuation_7], properties {dopplerIntensity=1.5}}
 {
  id=Enveloper_0
  children [SoundClass_0]
  properties {pitchCurve {editorCurveData {keys [{time=0}]}}, volumeCurve {editorCurveData {keys [<at 0 0>,<at 0.1 1>]}}}
 }
 {id=Mixer_0, children [QualityLevel_0, Enveloper_0], properties {inputVolume [0.8,1]}}
 {
  id=ModulatorContinuous_1
  children [WavePlayer_4]
  properties {
   pitchModulationParams {default=0, maxOutput=1.8, minOutput=1, parameterName=PreExplo}
   volumeModulationParams {default=0, minOutput=0.1, parameterName=PreExplo}
  }
 }
 {id=Modulator_1, children [Random_1]}
 {id=QualityLevel_0, children [Attenuation_0, Attenuation_0, Attenuation_0, null]}
 {id=Random_1, children [WavePlayer_0,WavePlayer_1,WavePlayer_2], properties {bShouldExcludeFromBranchCulling=true, weights [1,1,1]}}
 {id=SoundClass_0, children [Doppler_0], properties {soundClassOverride=/Game/Sounds/FortSoundClassesAndModes/Fort_Explosives_Medium_PSM}}
 {id=SoundClass_1, children [Modulator_1], properties {soundClassOverride=&Fort_Gadgets}}
 {id=WavePlayer_0, properties {soundWaveAssetPtr=&/DeployedCover_InAir_Start_01}}
 {id=WavePlayer_1, properties {soundWaveAssetPtr=&/DeployedCover_InAir_Start_02}}
 {id=WavePlayer_2, properties {soundWaveAssetPtr=&/DeployedCover_InAir_Start_03}}
 {id=WavePlayer_4, properties {bLooping=true, soundWaveAssetPtr=&/DeployedCover_Warning_01}}
]
properties {
 attenuationSettings=&/ATT_DeployedCover_InAirLoop
 maxDistance=22500
 soundClassObject=&Fort_WeaponRanged_Fire_3P_PSM
 volumeMultiplier=0.7
 attenuationOverrides {falloffDistance=5000, attenuationShapeExtents {x=1000}}
 concurrencyOverrides {bLimitToOwner=true, maxCount=3, resolutionRule=StopQuietest}
 concurrencySet [/Game/Sounds/Concurrency_Rules/Athena_Explosive_Fuse_Conc]
}
