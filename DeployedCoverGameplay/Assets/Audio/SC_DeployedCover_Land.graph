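# Sound cue for deployed cover landing, playing a single landing wave.
# The wave player feeds a Modulator that holds volume constant and leaves pitch at its default random range, and the cue lowers overall volume.
# It borrows weapon-drop attenuation and the loot-collect sound class, and joins the PickupMaximum_Oldest concurrency set.
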SoundCue /DeployedCoverGameplay/Assets/Audio/SC_DeployedCover_Land
root=Modulator_1
nodes [
 {id=Modulator_1, children [WavePlayer_0], properties {volumeMax=1, volumeMin=1}}
 {id=WavePlayer_0, properties {soundWaveAssetPtr=&/DeployedCover_Land}}
]
properties {
 attenuationSettings=&Fort_Weapon_Drop_Att
 maxDistance=3000
 soundClassObject=&Fort_Loot_Collect
 volumeMultiplier=0.6
 concurrencyOverrides {maxCount=1, resolutionRule=PreventNew}
 concurrencySet [&PickupMaximum_Oldest]
}
