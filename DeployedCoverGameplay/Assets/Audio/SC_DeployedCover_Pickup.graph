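# SoundCue that plays the pickup sound for the Deployed Cover item.
# A Random node picks with equal weight among three pickup wave variations.
# It uses weapon pickup attenuation and the loot collect sound class at reduced volume.
# A concurrency override allows one instance at a time and blocks new plays while one is active.
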SoundCue /DeployedCoverGameplay/Assets/Audio/SC_DeployedCover_Pickup
root=Random_0
nodes [
 {id=Random_0, children [WavePlayer_3,WavePlayer_4,WavePlayer_5], properties {weights [1,1,1]}}
 {id=WavePlayer_3, properties {soundWaveAssetPtr=&/DeployedCover_Pickup_B_01}}
 {id=WavePlayer_4, properties {soundWaveAssetPtr=&/DeployedCover_Pickup_B_02}}
 {id=WavePlayer_5, properties {soundWaveAssetPtr=&/DeployedCover_Pickup_B_03}}
]
properties {
 attenuationSettings=&Fort_Weapon_Pickup_Att
 bOverrideConcurrency=true
 maxDistance=3500
 soundClassObject=&Fort_Loot_Collect
 volumeMultiplier=0.5
 concurrencyOverrides {maxCount=1, resolutionRule=PreventNew}
}
