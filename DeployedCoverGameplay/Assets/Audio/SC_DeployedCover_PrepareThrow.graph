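# Sound cue for the foley played when a player prepares to throw Deployed Cover, classed as a gadget item sound.
# A quality-level root sends its first level to a random pick of six throwing waves, while the other three levels share a two-wave subset.
# Each random node feeds a single-input mixer that halves its volume.
# The cue uses gadget-placement attenuation and overrides concurrency to prevent new instances once at the limit.
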SoundCue /DeployedCoverGameplay/Assets/Audio/SC_DeployedCover_PrepareThrow
root=QualityLevel_0
nodes [
 {id=Mixer_0, children [Random_1], properties {inputVolume [0.5]}}
 {id=Mixer_1, children [Random_2], properties {inputVolume [0.5]}}
 {id=QualityLevel_0, children [Mixer_0,Mixer_1,Mixer_1,Mixer_1]}
 {id=Random_1, children [WavePlayer_0,WavePlayer_1,WavePlayer_2,WavePlayer_3,WavePlayer_5,WavePlayer_6], properties {weights [1,1,1,1,1,1]}}
 {id=Random_2, children [WavePlayer_8, WavePlayer_12], properties {weights [1,1]}}
 {id=WavePlayer_0, properties {soundWaveAssetPtr=&/DeployedCover_Foley_Throwing01}}
 {id=WavePlayer_1, properties {soundWaveAssetPtr=&/DeployedCover_Foley_Throwing02}}
 {id=WavePlayer_12, properties {soundWaveAssetPtr=&/DeployedCover_Foley_Throwing04}}
 {id=WavePlayer_2, properties {soundWaveAssetPtr=&/DeployedCover_Foley_Throwing03}}
 {id=WavePlayer_3, properties {soundWaveAssetPtr=&DeployedCover_Foley_Throwing04}}
 {id=WavePlayer_5, properties {soundWaveAssetPtr=&/DeployedCover_Foley_Throwing05}}
 {id=WavePlayer_6, properties {soundWaveAssetPtr=&/DeployedCover_Foley_Throwing06}}
 {id=WavePlayer_8, properties {soundWaveAssetPtr=&DeployedCover_Foley_Throwing02}}
]
properties {
 attenuationSettings=/Game/Sounds/Attenuation/Fort_GadgetPlace_Att
 bOverrideConcurrency=true
 maxDistance=2000
 soundClassObject=&Fort_Gadgets_Items
 volumeMultiplier=0.35
 concurrencyOverrides {resolutionRule=PreventNew}
}
