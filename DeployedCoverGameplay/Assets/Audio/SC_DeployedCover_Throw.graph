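# Sound cue for throwing the Deployed Cover gadget, classed as a gadget item sound and using gadget-placement attenuation.
# A mixer layers one quiet foley wave under an equally weighted random pick of three throw variations, then feeds a modulator.
# A quality-level root routes that same chain to its first three slots and leaves the fourth empty.
# Overridden concurrency prevents new instances from starting once the limit is reached.
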SoundCue /DeployedCoverGameplay/Assets/Audio/SC_DeployedCover_Throw
root=QualityLevel_0
nodes [
 {id=Mixer_0, children [WavePlayer_3,Random_0], properties {inputVolume [0.25,1]}}
 {id=Modulator_0, children [Mixer_0]}
 {id=QualityLevel_0, children [Modulator_0,Modulator_0,Modulator_0,null]}
 {id=Random_0, children [WavePlayer_0,WavePlayer_1,WavePlayer_2], properties {weights [1,1,1]}}
 {id=WavePlayer_0, properties {soundWaveAssetPtr=&/DeployedCover_Throw_01}}
 {id=WavePlayer_1, properties {soundWaveAssetPtr=&/DeployedCover_Throw_02}}
 {id=WavePlayer_2, properties {soundWaveAssetPtr=&/DeployedCover_Throw_03}}
 {id=WavePlayer_3, properties {soundWaveAssetPtr=&/DeployedCover_ThrowFoley_01}}
]
properties {
 attenuationSettings=/Game/Sounds/Attenuation/Fort_GadgetPlace_Att
 bOverrideConcurrency=true
 maxDistance=2000
 soundClassObject=&Fort_Gadgets_Items
 volumeMultiplier=0.5
 concurrencyOverrides {resolutionRule=PreventNew}
}
