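# Sound cue for the impact when a thrown deployed cover lands, layering an assemble sound with a stomp hit.
# A mixer plays one of two randomly chosen stomp impact waves at once, alongside the assemble wave, which is slightly delayed, pitch-varied and boosted.
# It uses the impacts sound class, small-object demolition attenuation with overrides, and a four-voice concurrency rule that stops the furthest or oldest.
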SoundCue /DeployedCoverGameplay/Assets/Audio/SC_DeployedCover_Throw_Impact
root=Mixer_0
nodes [
 {id=Delay_0, children [Modulator_0], properties {delayMax=0.07, delayMin=0.07}}
 {id=Mixer_0, children [Delay_0,Random_0], properties {inputVolume [1.2,1]}}
 {id=Modulator_0, children [WavePlayer_3], properties {pitchMax=1.02, pitchMin=0.98, volumeMin=1}}
 {id=Random_0, children [WavePlayer_0,WavePlayer_1], properties {weights [1,1]}}
 {id=WavePlayer_0, properties {soundWaveAssetPtr=&/DeployedCover_Stomp_Impact_01}}
 {id=WavePlayer_1, properties {soundWaveAssetPtr=&/DeployedCover_Stomp_Impact_02}}
 {id=WavePlayer_3, properties {soundWaveAssetPtr=&/DeployedCover_Assemble}}
]
properties {
 attenuationSettings=/Game/Sounds/Attenuation/Fort_Demolition_Hit_Small_Sized_Object_Att
 bOverrideAttenuation=true
 maxDistance=2600
 soundClassObject=&Fort_Impacts
 volumeMultiplier=1
 attenuationOverrides {falloffDistance=2400, stereoSpread=2000, attenuationShapeExtents {x=200}}
 concurrencyOverrides {maxCount=4}
 concurrencySet [/Game/Sounds/Concurrency_Rules/Demo_FourMaximum_StopFurthestOldest_VolumeScale]
}
