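; Center piece of the deployable cover gameplay item, a building actor that carries a separately spawned shielded window.
; On BeginPlay the server checks downward surface linking, fires a spawn cue and spawns the window directly, because dynamic build instructions only support Building Actors.
; Losing the linked actor destroys the piece, while death fires a cue and expires the window.
; A replicated death flag hides the mesh and disables collision.
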(event Custom|DEBUG_DeployStuff
  (Rendering|Material|SetScalarParameterValueonMaterials (Variables|Default|GetStaticMesh) "BP_GetTime" (Utilities|Time|GetGameTimeinSeconds)))

(event EventBeginPlay
  (|CallParentFunction)
  (Networking|SwitchHasAuthority
    (:Authority
      (Utilities|FlowControl|Sequence
        (:then_0
          (LinkToActorComponent|PerformLinkingCheck (Variables|Default|GetFortLinkToActor) "Down" (Variables|Default|GetLinkedSurfaceActor))
          (Ability|ExecuteGameplayCueReplicated (Variables|Default|GetGC_OnSpawn)))
        (:then_1
          (bind _returnvalue (Game|SpawnBuildingActorfromClass "/DeployedCoverGameplay/Items/DeployedActors/B_DeployedCover_ShieldedWindow.B_DeployedCover_ShieldedWindow_C" (Math|Transform|MakeTransform (+ (Variables|Default|GetWindowSpawnLocation) (.location (Transformation|GetActorTransform))) (.rotation (Transformation|GetActorTransform)) (.scale (Transformation|GetActorTransform))) 0 false 0 0)) ; dynamic build instructions only work for Building Actors, not for other BP types\nspawn the window this way instead
          (Variables|Default|SetSpawnedWindow _returnvalue))))))

(event OnLinkedActorDestroyed(FortLinkToActor)
  (params (DamageCauser :type Actor%20Object%20Reference))
  (Actor|DestroyActor)
  (Variables|Default|SetOnDeath? true))

(event EventOnDeathServer
  (params (Damage :type Float%20%28single-precision%29) (DamageTags :type Gameplay%20Tag%20Container%20Structure%20%28by%20ref%29) (Momentum :type Vector :default "0, 0, 0") (HitInfo :type Hit%20Result%20Structure%20%28by%20ref%29) (InstigatedBy :type Controller%20Object%20Reference) (DamageCauser :type Actor%20Object%20Reference) (EffectContext :type Gameplay%20Effect%20Context%20Handle%20Structure))
  (Utilities|FlowControl|Sequence
    (:then_0
      (Variables|Default|SetOnDeath? true)
      (Ability|ExecuteGameplayCueReplicated (Variables|Default|GetGC_OnDeath))
      (bind _spawnedwindow (Variables|Default|GetSpawnedWindow)
        (:then
          (Utilities|FlowControl|DoOnce false
            (:Completed
              (Actor|SetLifeSpan _spawnedwindow 0.1)
              (Variables|/DeployedCoverGameplay/Items/DeployedActors/B_DeployedCover_ShieldedWindow.B_DeployedCover_ShieldedWindow_C|SetbHasBeenDestroyed true _spawnedwindow))))
        (:else)))
    (:then_1
      (|CallParentFunction Damage DamageTags Momentum HitInfo InstigatedBy DamageCauser EffectContext))))

(fn OnRep_OnDeath?
  (if (Variables|Default|GetOnDeath?)
    (Development|SetHiddeninGame (Variables|Default|GetStaticMesh) true false)
    (Collision|SetActorEnableCollision false)))

(fn ConstructionScript
  (|CallParentFunction))
