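; Right-side piece of a deployable cover, judging by its name, that plays spawn and death effects and is removed along with its linked actor.
; BeginPlay and server death execute replicated gameplay cues.
; Death or linked-actor destruction sets OnDeath?, whose RepNotify hides the mesh and disables collision; linked destruction also destroys the actor.
; A debug event writes game time into the mesh materials' BP_GetTime parameter.
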(event Custom|DEBUG_DeployStuff
  (Rendering|Material|SetScalarParameterValueonMaterials (Variables|Default|GetStaticMesh) "BP_GetTime" (Utilities|Time|GetGameTimeinSeconds)))

(event EventBeginPlay
  (|CallParentFunction)
  (Ability|ExecuteGameplayCueReplicated (Variables|Default|GetGC_OnSpawn)))

(event OnLinkedActorDestroyed(FortLinkToActor)
  (params (DamageCauser :type Actor%20Object%20Reference))
  (Actor|DestroyActor)
  (Variables|Default|SetOnDeath? true))

(event EventOnDeathServer
  (params (Damage :type Float%20%28single-precision%29) (DamageTags :type Gameplay%20Tag%20Container%20Structure%20%28by%20ref%29) (Momentum :type Vector :default "0, 0, 0") (HitInfo :type Hit%20Result%20Structure%20%28by%20ref%29) (InstigatedBy :type Controller%20Object%20Reference) (DamageCauser :type Actor%20Object%20Reference) (EffectContext :type Gameplay%20Effect%20Context%20Handle%20Structure))
  (Variables|Default|SetOnDeath? true)
  (Ability|ExecuteGameplayCueReplicated (Variables|Default|GetGC_OnDeath)))

(fn OnRep_OnDeath?
  (if (Variables|Default|GetOnDeath?)
    (Development|SetHiddeninGame (Variables|Default|GetStaticMesh) true false)
    (Collision|SetActorEnableCollision false)))

(fn ConstructionScript
  (|CallParentFunction))
