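# Parent material for deployable cover pieces, animating their deployment through world position offset timed by Blueprint-set start parameters.
# A layered GigaMix base feeds shared damage, wind, snow, DirectFX, biome and RVT blending functions, with a simplified mobile low-quality path.
# Vertex red masks separate ramps, which rotate about a pivot with updated normals, from sliders, which translate.
# Pivot scale-in and stomp impact offsets are also computed, and IsWindow switches to open/close window motion.
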Material /DeployedCoverGameplay/Assets/Materials/M_Gameplay_DeployedCover_PARENT
graph {
 properties {
  bAutomaticallySetUsageInEditor=false
  bTangentSpaceNormal=false
  bUseMaterialAttributes=true
  bUsedWithInstancedStaticMeshes=true
  bUsedWithNanite=true
  bUsedWithSplineMeshes=true
  ditheredLODTransition=true
  maxWorldPositionOffsetDisplacement=100000
  pixelDepthOffsetMode=PDOM_Legacy
  shadingModel=MSM_FromMaterialExpression
  usedShadingModels=Default Lit | Subsurface | Clear Coat | Two Sided Foliage
 }
 nodes [
  n1 = MaterialFunctionCall() {materialFunction=/Game/Global/Base4Layers_2022/MaterialFunctions/MF_GigaMix_2023}
  MaterialExpressionReroute_70 = Reroute(None=@n1.Result)
  n2 = ScalarParameter() {
   desc=Documentation link
   group="0 (Expand)  Documentation (RMB - Copy display name) and Debug view"
   parameterName=https://confluence.it.epicgames.com/display/FORT/Fortnite+Base+Material+2022
  }
  n3 = SetMaterialAttributes!(MaterialAttributes=@MaterialExpressionReroute_70.MaterialAttributes, World Position Offset=Multiply(A=@n2)) {attributeSetTypes [#1]}
  MaterialExpressionReroute_71 = Reroute(None=@n3)
  MaterialExpressionReroute_69 = Reroute(None=@MaterialExpressionReroute_71.Output)
  n4 = MaterialFunctionCall(Material_In=@MaterialExpressionReroute_69.MaterialAttributes) {materialFunction=&MF_FN_DestructibleDamage}
  n5 = MaterialFunctionCall!(
   In=MaterialFunctionCall!(Material Attributes=MaterialFunctionCall!(In=@n4.Material_Out) {materialFunction=/Game/Athena/WinterArt/Materials/MF_WPO_WindNew}.Result) {materialFunction=/Game/Creative/Sets/MilitaryBase/Props/MF_FanSpin}.Result
   UseSphericalMappingOverride
  ) {materialFunction=/Game/Packages/Fortress_SharedMaterials/Base_Material_Functions/MF_CurieFallbackEffect}
  n6 = MaterialFunctionCall!(In=MaterialFunctionCall!(In=MaterialFunctionCall!(In=@n5.Result, Custom Mask, Custom Rotation Axis) {materialFunction=/Game/Athena/WinterArt/Materials/MF_WindWPO_v2}.Result) {materialFunction=/Game/Global/Base4Layers_2022/MaterialFunctions/MF_SnowAndIceArchitecture}.Result) {materialFunction=/Game/Global/Base4Layers_2022/MaterialFunctions/MF_VibrationRattle}
  MaterialExpressionReroute_68 = Reroute(None=@n6.Result)
  Composite!(None=@MaterialExpressionReroute_68.Result) {desc=Transform to World in case we will use RVT after this point, subgraphName=Tangent to World}
  Composite!(MaterialAttributes=@n1.Result) {subgraphName=Documentation Link}
  Composite(MaterialAttributes=@MaterialExpressionReroute_71.Output) {desc=Game Shared WPO and Fire Effect, subgraphName=WPO And Shared Functions}
  MaterialExpressionReroute_1 = Reroute(None=@n13.RVT H)
  MaterialExpressionReroute_3 = Reroute(None=@MaterialExpressionReroute_68.Result)
  n7 = GetMaterialAttributes(None=@MaterialExpressionReroute_3.Input) {attributeGetTypes [#2]}
  n8 = Transform(None=@n7.Normal)
  n9 = LinearInterpolate(A=Multiply(A=@n8) {constB=-1}, B=@n8, Alpha=Saturate(None=TwoSidedSign())) {desc="Fixes normals\r\non backfaces"}
  n10 = SetMaterialAttributes!(
   MaterialAttributes=@n7.MaterialAttributes
   Normal=StaticSwitchParameter(True=MaterialProxyReplace(Realtime=@n9, MaterialProxy=@n7.Normal), False=@n9) {parameterName=Force TangentSpaceNormalsFor HLOD ProxyReplace}
  ) {attributeSetTypes [#2]}
  MaterialExpressionReroute_2 = Reroute(None=@n10)
  n11 = MaterialFunctionCall!(In=MaterialFunctionCall!(In=@MaterialExpressionReroute_2.Output) {materialFunction=/Game/ArtTools/Tech_Art/DirectFX/Assets/MaterialFunctions/MF_DirectFX}.Result) {materialFunction=/Game/ArtTools/Tech_Art/Veneer/MaterialFunctions/MF_Veneer}
  MaterialExpressionReroute_8 = Reroute(None=@n11.Result)
  n12 = MaterialFunctionCall!(AO, RVT_Height=@MaterialExpressionReroute_1."RVT_Height (S)", For Buildings=@MaterialExpressionReroute_8."For Moss (MA)", For Moss) {materialFunction=/Game/Environments/Asteria/Rocks/Materials/MaterialFunctions/MF_JungleBiome}
  MaterialExpressionReroute_0 = Reroute(None=@n12.Buildings Jungl)
  n13 = MaterialFunctionCall!(
   In=@MaterialExpressionReroute_0.Result
   "RVT Height Map (Roughness)"=GetMaterialAttributes(None=@n11.Result) {attributeGetTypes [#3]}.Roughness
  ) {materialFunction=/Game/Global/Base4Layers_2022/MaterialFunctions/MF_RVTBlending}
  Composite!(InputPin=@n13.Result) {subgraphName=Mobile Opt}
  Composite!("For Moss (MA)"=@n11.Result, "RVT_Height (S)"=@n13.RVT H) {subgraphName="Seasonal "}
  n14 = VertexColor()
  n15 = Add(A=@n14.R) {constB=-0.6}
  NamedRerouteDeclaration(None=Saturate(None=Multiply(A=@n15) {constB=10})) {name=Mask_Sliders, nodeColor=<rgba 0 0.2969 1 1>}
  n16 = Multiply(A=OneMinus(None=Add(A=@n14.R) {constB=0.1})) {constB=50}
  NamedRerouteDeclaration(None=Saturate(None=Multiply(A=@n14.R, B=@n16))) {name=Mask_Ramps, nodeColor=<rgba 0.6172 1 0 1>}
  n17 = NamedRerouteUsage()
  n18 = Constant3Vector() {constant=<rgba 0 0 200 0>}
  n19 = VectorParameter() {group=WPO_Animation, parameterName=Scale_Pivot}
  n20 = Subtract(A=Add(A=@n17, B=WorldPosition().XYZ), B=TransformPosition(None=Add(A=@n18, B=@n19.RGB)) {transformType=TRANSFORMPOSSOURCE_World})
  n21 = Subtract(
   A=Subtract(A=Time(), B=ScalarParameter() {defaultValue=1.0000000200408773e+21, group=WPO_Animation, parameterName=BP_GetTime_Scale, sortPriority=100})
   B=ScalarParameter() {group=WPO_Animation, parameterName=Scale_Delay}
  )
  n22 = Multiply(A=ScalarParameter() {defaultValue=5, group=WPO_Animation, parameterName=Scale_Speed}) {constB=0.5}
  n23 = NamedRerouteDeclaration(None=Saturate(None=Multiply(A=@n21, B=@n22))) {name=ScaleAnimationFactor, nodeColor=<rgba 0 0.4844 1 1>}
  n24 = OneMinus(None=@n23)
  n25 = Multiply(A=@n24, B=@n24)
  n26 = Multiply(A=Multiply(A=@n25, B=@n25), B=@n24)
  n27 = OneMinus(None=LinearInterpolate(A=ScalarParameter() {defaultValue=0.5, group=WPO_Animation, parameterName=All_InitialScale}, Alpha=OneMinus(None=@n26)))
  NamedRerouteDeclaration(None=Subtract(A=@n17, B=Multiply(A=@n20, B=@n27))) {name=WPO_Scaled, nodeColor=<rgba 0 0.3437 1 1>}
  n28 = StaticSwitchParameter(True=NamedRerouteUsage(), False=NamedRerouteUsage()) {group=WPO_Animation, parameterName=IsWindow, sortPriority=-1}
  NamedRerouteDeclaration(None=@n28) {name=WPO_NoScale, nodeColor=<rgba 0.9453 1 0 1>}
  n29 = Subtract(A=Time(), B=ScalarParameter() {defaultValue=1.0000000200408773e+21, group=WPO_Animation, parameterName=BP_GetTime_Scale, sortPriority=100})
  n30 = ScalarParameter() {defaultValue=0.2, group=WPO_Animation, parameterName=Impact_Delay}
  n31 = Saturate(None=Multiply(A=Subtract(A=@n29, B=@n30), B=ScalarParameter() {defaultValue=4.5, parameterName=Impact_Speed}))
  n32 = Multiply(A=@n31, B=@n31)
  NamedRerouteDeclaration(None=LinearInterpolate(
    A=VectorParameter() {group=WPO_Animation, parameterName=Impact_InitialPosition, defaultValue=<rgba 0 0 60 0>}.RGB
    Alpha=Multiply(A=@n32, B=@n32)
   ) {constB=0}) {name=InitialStompImpact, nodeColor=<rgba 0 0.4609 1 1>}
  n33 = LinearInterpolate(Alpha=NamedRerouteUsage()) {constA=0.3, constB=0.5}
  n34 = NamedRerouteDeclaration(None=@n33) {name=FirstAction_Timing, nodeColor=<rgba 0.0078 0 1 1>}
  n35 = NamedRerouteUsage()
  n36 = Saturate(None=InverseLinearInterpolate(B=@n34, Value=@n35) {constB=0.5})
  n37 = Add(A=@n35, B=OneMinus(None=@n34))
  n38 = LinearInterpolate(A=Multiply(A=Multiply(A=@n36, B=@n36), B=@n36), Alpha=Saturate(None=Floor(None=@n37)))
  n39 = NamedRerouteUsage()
  n40 = ScalarParameter() {group=WPO_Animation, parameterName=AnimationDelay_Ramps, sortPriority=50}
  n41 = ScalarParameter() {group=WPO_Animation, parameterName=AnimationDelay_Sliders, sortPriority=50}
  n42 = NamedRerouteUsage()
  n43 = Subtract(
   A=Subtract(A=Time(), B=LinearInterpolate(A=@n40, B=@n41, Alpha=@n42))
   B=ScalarParameter() {defaultValue=1.0000000200408773e+21, group=WPO_Animation, parameterName=BP_GetTime, sortPriority=100}
  )
  n44 = LinearInterpolate(
   A=Multiply(A=ScalarParameter() {defaultValue=1, group=WPO_Animation, parameterName=AnimationSpeed_Ramps, sortPriority=50}) {constB=0.6}
   B=Multiply(A=ScalarParameter() {defaultValue=1, group=WPO_Animation, parameterName=AnimationSpeed_Sliders, sortPriority=50}) {constB=0.8}
   Alpha=NamedRerouteUsage()
  )
  n45 = NamedRerouteDeclaration(None=Saturate(None=Multiply(A=@n43, B=@n44))) {name=Ramps_NormalizedTime, nodeColor=<rgba 0.9688 0 1 1>}
  n46 = Saturate(None=InverseLinearInterpolate(A=NamedRerouteUsage(), B=NamedRerouteUsage(), Value=@n45) {constA=0.5, constB=0.8})
  n47 = Multiply(A=@n46, B=@n46)
  n48 = LinearInterpolate(
   B=NamedRerouteDeclaration(None=LinearInterpolate(Alpha=@n39) {constA=0.9, constB=0.95}) {name=BounceAmount, nodeColor=<rgba 0 0.9531 1 1>}
   Alpha=Multiply(A=@n47, B=@n47)
  ) {constA=1, constB=0.8}
  n49 = NamedRerouteUsage()
  n50 = NamedRerouteUsage()
  n51 = Multiply(A=NamedRerouteUsage()) {constB=2}
  n52 = LinearInterpolate(A=Subtract(A=Subtract(A=@n51), B=@n50), Alpha=NamedRerouteUsage()) {constA=0.3, constB=0}
  n53 = OneMinus(None=NamedRerouteDeclaration(None=Add(A=@n50, B=@n52)) {name=SecondAction_Timing, nodeColor=<rgba 0.2161 1 0.6102 1>})
  n54 = Saturate(None=Floor(None=Add(A=@n49, B=@n53)))
  n55 = NamedRerouteUsage()
  n56 = NamedRerouteDeclaration(None=LinearInterpolate(Alpha=NamedRerouteUsage()) {constA=0.8, constB=0.6}) {name=ThirdAction_Timing, nodeColor=<rgba 0.67 1 1 1>}
  n57 = Saturate(None=InverseLinearInterpolate(A=@n56, Value=@n45) {constA=0.8})
  n58 = Multiply(A=@n57, B=@n57)
  n59 = Multiply(A=@n58, B=@n58)
  n60 = Floor(None=Add(A=OneMinus(None=@n56), B=NamedRerouteUsage()))
  NamedRerouteDeclaration(None=LinearInterpolate(A=LinearInterpolate(A=@n38, B=@n48, Alpha=@n54), B=LinearInterpolate(A=@n55, Alpha=@n59) {constA=0.8}, Alpha=Saturate(None=@n60))) {name=AnimationTime, nodeColor=<rgba 1 0 0.9375 1>}
  n61 = ScalarParameter() {group=WPO_Animation, parameterName=Ramp_InitialRotationAmount}
  n62 = Divide(A=@n61) {constB=8}
  n63 = Saturate(None=NamedRerouteUsage())
  NamedRerouteDeclaration(None=Multiply(A=LinearInterpolate(A=@n61, B=@n62, Alpha=@n63) {constB=-0.05}, B=NamedRerouteUsage())) {name=RampRotation, nodeColor=<rgba 0 1 0.7422 1>}
  n64 = Multiply(
   A=LinearInterpolate(A=VectorParameter() {group=WPO_Animation, parameterName=Sliders_InitialPosition}.RGB, Alpha=NamedRerouteUsage()) {constB=0}
   B=NamedRerouteUsage()
  )
  n65 = NamedRerouteUsage()
  NamedRerouteDeclaration(None=Transform(None=Add(A=@n64, B=@n65)) {transformSourceType=TRANSFORMSOURCE_Local}) {name=SlidersPositionOffset, nodeColor=<rgba 1 0.4219 0 1>}
  n66 = ScalarParameter() {group=WPO_Animation, parameterName=BP_WPO_Window_CLOSE/OPEN, sortPriority=100}
  n67 = Subtract(A=Time(), B=ScalarParameter() {group=WPO_Animation, parameterName=AnimationDelay_Window, sortPriority=50})
  n68 = ScalarParameter() {defaultValue=1.0000000200408773e+21, group=WPO_Animation, parameterName=BP_GetTime, sortPriority=100}
  n69 = ScalarParameter() {defaultValue=1, group=WPO_Animation, parameterName=AnimationSpeed_Window, sortPriority=50}
  n70 = Multiply(A=@n69) {constB=3.2}
  n71 = Subtract(A=Time(), B=ScalarParameter() {defaultValue=1.0000000200408773e+21, group=WPO_Animation, parameterName=BP_GetTime_Scale, sortPriority=100})
  n72 = Floor(None=Saturate(None=Subtract(A=@n71) {constB=0.75}))
  n73 = Saturate(None=Multiply(A=Subtract(A=@n67, B=@n68), B=LinearInterpolate(A=@n69, B=@n70, Alpha=@n72)))
  n74 = OneMinus(None=@n73)
  n75 = LinearInterpolate(A=Multiply(A=@n73, B=@n73), B=OneMinus(None=Multiply(A=@n74, B=@n74)), Alpha=@n73)
  NamedRerouteDeclaration(None=Abs(None=Subtract(A=@n66, B=@n75))) {name=AnimationTime_Window, nodeColor=<rgba 0.5 0 1 1>}
  n76 = LinearInterpolate(A=VectorParameter() {group=WPO_Animation, parameterName=Window_InitialPosition}.RGB, Alpha=NamedRerouteUsage()) {constB=0}
  n77 = NamedRerouteUsage()
  NamedRerouteDeclaration(None=Transform(None=Add(A=@n76, B=@n77)) {transformSourceType=TRANSFORMSOURCE_Local}) {name=WindowPositionOffset, nodeColor=<rgba 0.3783 0.8847 1 1>}
  MaterialExpressionReroute_7 = Reroute(None=@n13.Result)
  n78 = GetMaterialAttributes(None=@MaterialExpressionReroute_7.InputPin)
  n79 = Constant() {r=0.5}
  n80 = Constant() {r=0.8}
  n81 = ShadingPathSwitch(
   Default=@MaterialExpressionReroute_7.InputPin
   Mobile=QualitySwitch(
    Default=@MaterialExpressionReroute_7.InputPin
    Low=SetMaterialAttributes!(MaterialAttributes=@n78.MaterialAttributes, Specular=@n79, Roughness=@n80) {attributeSetTypes [#4, #3]}
    High=@MaterialExpressionReroute_7.InputPin
    Medium=@MaterialExpressionReroute_7.InputPin
    Epic=@MaterialExpressionReroute_7.InputPin
   )
  )
  MaterialExpressionReroute_6 = Reroute(None=@n81)
  n82 = GetMaterialAttributes(None=MaterialFunctionCall(MA In=@MaterialExpressionReroute_6.Output) {materialFunction=&MF_CreativeMoveTool}.MA Out) {attributeGetTypes [#2]}
  NamedRerouteDeclaration(None=@n82.Normal) {name=OriginalNormal, nodeColor=<rgba 0 1 0.8125 1>}
  n83 = VectorParameter() {group=WPO_Animation, parameterName=Ramp_RotationAxis, defaultValue=<rgba 1 0 0 0>}
  n84 = MaterialFunctionCall!(
   RotationAngle=NamedRerouteUsage()
   NormalizedRotationAxis=Normalize(VectorInput=Transform(None=@n83.RGB) {transformSourceType=TRANSFORMSOURCE_Local})
   PivotPoint=TransformPosition(None=VectorParameter() {group=WPO_Animation, parameterName=Ramp_PivotPoint}.RGB) {transformType=TRANSFORMPOSSOURCE_World}
   Position
   Normal=NamedRerouteUsage()
   TangentU
   TangentV
  ) {materialFunction=/Game/Characters/Player/Male/Medium/Bodies/M_MED_Inferno/FX/Mat/RotateVertexAboutAxis}
  NamedRerouteDeclaration(None=@n84.Normal) {name=UpdatedNormal, nodeColor=<rgba 1 0.4922 0 1>}
  NamedRerouteDeclaration(None=Add(A=@n84.WPO, B=NamedRerouteUsage())) {name="Ramps&&Sliders_WPO", nodeColor=<rgba 0.1719 0 1 1>}
  n85 = Subtract(B=@n23)
  n86 = NamedRerouteUsage()
  n87 = ScalarParameter() {defaultValue=20, group=WPO_Animation, parameterName=Scale_OvershootingConstant}
  n88 = Multiply(A=@n86, B=@n87)
  n89 = Divide(A=Multiply(B=Sine(None=@n88)) {constA=2}, B=@n87)
  n90 = Cosine(None=Add(A=@n86, B=@n87))
  OneMinus(None=Multiply(A=Multiply(A=@n85, B=@n85), B=Add(A=@n89, B=@n90)))
  PinBase() {reroutePins [Input {expression=&self:MaterialExpressionReroute_3}]}
  PinBase!(Output=@n10) {reroutePins [Output {expression=&self:MaterialExpressionReroute_2}]}
  PinBase!() {reroutePins [
    "For Moss (MA)" {expression=&self:MaterialExpressionReroute_8}
    "RVT_Height (S)" {expression=&self:MaterialExpressionReroute_1}
   ]}
  PinBase!(Result=@n12.Buildings Jungl) {reroutePins [Result {expression=&self:MaterialExpressionReroute_0}]}
  PinBase!() {reroutePins [MaterialAttributes {expression=&self:MaterialExpressionReroute_70}]}
  PinBase!(Output=@n3) {reroutePins [Output {expression=&self:MaterialExpressionReroute_71}]}
  PinBase!() {reroutePins [MaterialAttributes {expression=&self:MaterialExpressionReroute_69}]}
  PinBase!(Result=@n6.Result) {reroutePins [Result {expression=&self:MaterialExpressionReroute_68}]}
  PinBase!() {reroutePins [InputPin {expression=&self:MaterialExpressionReroute_7}]}
  PinBase!(Output=@n81) {reroutePins [Output {expression=&self:MaterialExpressionReroute_6}]}
  Reroute(None=@n33)
  Reroute(None=@n28)
  SetMaterialAttributes!(
   MaterialAttributes=@n82.MaterialAttributes
   World Position Offset=NamedRerouteUsage()
   Normal=StaticSwitchParameter(True=NamedRerouteUsage(), False=NamedRerouteUsage()) {group=WPO_Animation, parameterName=IsWindow, sortPriority=-1}
  ) {attributeSetTypes [#1, #2]}
  Transform(None=@n20) {transformSourceType=TRANSFORMSOURCE_World, transformType=TRANSFORM_Local}
 ]
}
