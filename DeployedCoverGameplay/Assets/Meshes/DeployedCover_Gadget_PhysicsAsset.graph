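# Physics asset for the DeployedCover gadget mesh, giving the deployable cover gadget its collision and physics shape.
# A single body on the root bone uses one capsule rotated onto its side.
# It has no constraints, so the gadget simulates as one rigid piece with no articulated parts.
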PhysicsAsset /DeployedCoverGameplay/Assets/Meshes/DeployedCover_Gadget_PhysicsAsset
bodies {root [{shapeType=Capsule, length=7.1729, radius=6.1855, rotation=<rot 0 -90 0>}]}
constraints []
